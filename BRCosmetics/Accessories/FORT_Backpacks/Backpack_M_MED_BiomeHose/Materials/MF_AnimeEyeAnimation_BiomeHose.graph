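# Material function that animates the BiomeHose backpack's stylized eyes, compositing them over incoming lit and shaded colors.
# Pose parameters for gaze and per-eye blink offset and scale UVs from a selectable channel, with per-eye masks handling left and right.
# An eye mask texture supplies pupil and lid shapes, with sphere masks and smoothsteps drawing lid edges and optional outlines.
# Static switches toggle the animation and outline overlap.
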MaterialFunction /BRCosmetics/Accessories/FORT_Backpacks/Backpack_M_MED_BiomeHose/Materials/MF_AnimeEyeAnimation_BiomeHose
graph {expressions {
  MaterialExpressionAbs_4 {class=MaterialExpressionAbs, inputs [None {from=MaterialExpressionAdd_35, output=""}], properties {}}
  MaterialExpressionAbs_6 {class=MaterialExpressionAbs, inputs [None {from=MaterialExpressionAdd_36, output=""}], properties {}}
  MaterialExpressionAbs_8 {class=MaterialExpressionAbs, inputs [None {from=MaterialExpressionAdd_37, output=""}], properties {}}
  MaterialExpressionAbs_9 {class=MaterialExpressionAbs, inputs [None {from=MaterialExpressionAdd_38, output=""}], properties {}}
  MaterialExpressionAdd_1 {
   class=MaterialExpressionAdd
   inputs [
    A {from=MaterialExpressionReroute_2}
    B {from=MaterialExpressionLinearInterpolate_2}
   ]
   properties {constB=0}
  }
  MaterialExpressionAdd_11 {
   class=MaterialExpressionAdd
   inputs [
    A {from=MaterialExpressionScalarParameter_20}
    B {from=MaterialExpressionMultiply_13}
   ]
  }
  MaterialExpressionAdd_15 {
   class=MaterialExpressionAdd
   inputs [
    A {from=MaterialExpressionMaterialFunctionCall_2, output=R}
    B {from=""}
   ]
   properties {constB=-1}
  }
  MaterialExpressionAdd_16 {
   class=MaterialExpressionAdd
   inputs [
    A {from=MaterialExpressionMaterialFunctionCall_16, output=G}
    B {from=""}
   ]
   properties {constB=-0.5}
  }
  MaterialExpressionAdd_17 {
   class=MaterialExpressionAdd
   inputs [
    A {from=MaterialExpressionMultiply_5}
    B {from=""}
   ]
   properties {constB=0.5}
  }
  MaterialExpressionAdd_18 {
   class=MaterialExpressionAdd
   inputs [
    A {from=MaterialExpressionMultiply_38}
    B {from=""}
   ]
   properties {constB=0.5}
  }
  MaterialExpressionAdd_2 {
   class=MaterialExpressionAdd
   inputs [
    A {from=MaterialExpressionMultiply_19}
    B {from=MaterialExpressionLinearInterpolate_1}
   ]
   properties {constB=0}
  }
  MaterialExpressionAdd_20 {
   class=MaterialExpressionAdd
   inputs [
    A {from=MaterialExpressionNamedRerouteUsage_24}
    B {from=""}
   ]
   properties {constB=0}
  }
  MaterialExpressionAdd_22 {
   class=MaterialExpressionAdd
   inputs [
    A {from=MaterialExpressionMultiply_35}
    B {from=""}
   ]
   properties {constB=0.5}
  }
  MaterialExpressionAdd_24 {
   class=MaterialExpressionAdd
   inputs [
    A {from=MaterialExpressionMaterialFunctionCall_5, output=G}
    B {from=""}
   ]
   properties {constB=0}
  }
  MaterialExpressionAdd_25 {
   class=MaterialExpressionAdd
   inputs [
    A {from=MaterialExpressionNamedRerouteUsage_21}
    B {from=""}
   ]
   properties {constB=-0.5, desc=LeftEyeUVs}
  }
  MaterialExpressionAdd_35 {
   class=MaterialExpressionAdd
   inputs [
    A {from=MaterialExpressionAdd_7}
    B {from=""}
   ]
   properties {constB=-0.5}
  }
  MaterialExpressionAdd_36 {
   class=MaterialExpressionAdd
   inputs [
    A {from=MaterialExpressionAdd_1}
    B {from=""}
   ]
   properties {constB=-0.5}
  }
  MaterialExpressionAdd_37 {
   class=MaterialExpressionAdd
   inputs [
    A {from=MaterialExpressionAdd_2}
    B {from=""}
   ]
   properties {constB=-0.5}
  }
  MaterialExpressionAdd_38 {
   class=MaterialExpressionAdd
   inputs [
    A {from=MaterialExpressionAdd_8}
    B {from=""}
   ]
   properties {constB=-0.5}
  }
  MaterialExpressionAdd_4 {
   class=MaterialExpressionAdd
   inputs [
    A {from=MaterialExpressionMaterialFunctionCall_17, output=Result}
    B {from=MaterialExpressionMultiply_27}
   ]
   properties {desc=move right eye to UV space 0-1}
  }
  MaterialExpressionAdd_5 {
   class=MaterialExpressionAdd
   inputs [
    A {from=MaterialExpressionAdd_4}
    B {from=MaterialExpressionAppendVector_0}
   ]
  }
  MaterialExpressionAdd_7 {
   class=MaterialExpressionAdd
   inputs [
    A {from=MaterialExpressionMultiply_18}
    B {from=MaterialExpressionLinearInterpolate_3}
   ]
   properties {constB=0}
  }
  MaterialExpressionAdd_8 {
   class=MaterialExpressionAdd
   inputs [
    A {from=MaterialExpressionReroute_1}
    B {from=MaterialExpressionLinearInterpolate_4}
   ]
   properties {constB=0}
  }
  MaterialExpressionAdd_9 {
   class=MaterialExpressionAdd
   inputs [
    A {from=MaterialExpressionScalarParameter_26}
    B {from=MaterialExpressionMultiply_16}
   ]
  }
  MaterialExpressionAppendVector_0 {
   class=MaterialExpressionAppendVector
   inputs [
    A {from=MaterialExpressionMultiply_40, output=""}
    B {from=MaterialExpressionMultiply_41, output=""}
   ]
  }
  MaterialExpressionAppendVector_2 {
   class=MaterialExpressionAppendVector
   inputs [
    A {from=MaterialExpressionLinearInterpolate_19, output=""}
    B {from=MaterialExpressionConstant_4, output=""}
   ]
  }
  MaterialExpressionCeil_1 {class=MaterialExpressionCeil, inputs [None {from=MaterialExpressionMaterialFunctionCall_4, output=G}], properties {}}
  MaterialExpressionClamp_2 {
   class=MaterialExpressionClamp
   inputs [
    None {from=MaterialExpressionAdd_9, output=""}
    Min {from="", output=""}
    Max {from="", output=""}
   ]
   properties {minDefault=-1}
  }
  MaterialExpressionClamp_3 {
   class=MaterialExpressionClamp
   inputs [
    None {from=MaterialExpressionAdd_11, output=""}
    Min {from="", output=""}
    Max {from="", output=""}
   ]
   properties {minDefault=-1}
  }
  MaterialExpressionConstant2Vector_0 {class=MaterialExpressionConstant2Vector, properties {r=-1}}
  MaterialExpressionConstant_14 {class=MaterialExpressionConstant}
  MaterialExpressionConstant_17 {class=MaterialExpressionConstant, properties {r=2}}
  MaterialExpressionConstant_18 {class=MaterialExpressionConstant, properties {r=4}}
  MaterialExpressionConstant_19 {class=MaterialExpressionConstant, properties {r=0.2}}
  MaterialExpressionConstant_4 {class=MaterialExpressionConstant, properties {r=1}}
  MaterialExpressionConstant_7 {class=MaterialExpressionConstant, properties {r=1}}
  MaterialExpressionConvert_0 {
   class=MaterialExpressionConvert
   inputs [
    None {from=MaterialExpressionScalarParameter_40}
    None {from=""}
   ]
   properties {
    nodeName=Make Vector2
    convertInputs [
     {
      type=Scalar
      expressionInput {
       expression=&self:MaterialExpressionScalarParameter_40
       inputName=None
       mask=0
       maskA=0
       maskB=0
       maskG=0
       maskR=0
       outputIndex=0
      }
     }
     {type=Scalar, expressionInput {expression=null, inputName=None, mask=0, maskA=0, maskB=0, maskG=0, maskR=0, outputIndex=0}}
    ]
    convertOutputs [{type=Vector2}]
   }
  }
  MaterialExpressionDistance_2 {
   class=MaterialExpressionDistance
   inputs [
    A {from=MaterialExpressionTextureSample_2, output=G}
    B {from=MaterialExpressionConstant_19, output=""}
   ]
   properties {}
  }
  MaterialExpressionFloor_0 {class=MaterialExpressionFloor, inputs [None {from=MaterialExpressionAdd_20, output=""}], properties {}}
  MaterialExpressionFloor_1 {class=MaterialExpressionFloor, inputs [None {from=MaterialExpressionMaterialFunctionCall_4, output=R}], properties {desc=LeftEye}}
  MaterialExpressionFunctionInput_0 {
   class=MaterialExpressionFunctionInput
   inputs [Preview {from="", output=""}]
   properties {bUsePreviewValueAsDefault=true, inputName=Mask, inputType=FunctionInput_Scalar, sortPriority=20, previewValue=<1 0 0 1>}
  }
  MaterialExpressionFunctionInput_1 {
   class=MaterialExpressionFunctionInput
   inputs [Preview {from=MaterialExpressionStaticBool_0, output=""}]
   properties {
    bUsePreviewValueAsDefault=true
    description="if true (default), uses the second UV channel for eye mapping. If false, uses the first UV channel."
    inputName=Use UV2
    inputType=FunctionInput_StaticBool
    sortPriority=15
    previewValue=<1 0 0 1>
   }
  }
  MaterialExpressionFunctionInput_2 {class=MaterialExpressionFunctionInput, inputs [Preview {from="", output=""}], properties {inputName=Color Lit, sortPriority=1}}
  MaterialExpressionFunctionInput_3 {class=MaterialExpressionFunctionInput, inputs [Preview {from="", output=""}], properties {inputName=Color Shaded, sortPriority=2}}
  MaterialExpressionFunctionOutput_1 {
   class=MaterialExpressionFunctionOutput
   inputs [None {from=MaterialExpressionStaticSwitchParameter_5, output=""}]
   properties {outputName=Lit, sortPriority=1}
  }
  MaterialExpressionFunctionOutput_2 {
   class=MaterialExpressionFunctionOutput
   inputs [None {from=MaterialExpressionStaticSwitchParameter_4, output=""}]
   properties {outputName=Shaded, sortPriority=2}
  }
  MaterialExpressionLinearInterpolate_1 {
   class=MaterialExpressionLinearInterpolate
   inputs [
    A {from=MaterialExpressionScalarParameter_11}
    B {from=MaterialExpressionScalarParameter_6}
    Alpha {from=MaterialExpressionNamedRerouteDeclaration_16}
   ]
   properties {constB=0.6}
  }
  MaterialExpressionLinearInterpolate_12 {
   class=MaterialExpressionLinearInterpolate
   inputs [
    A {from=MaterialExpressionVectorParameter_0, output=RGB}
    B {from=MaterialExpressionLinearInterpolate_6}
    Alpha {from=MaterialExpressionNamedRerouteUsage_44}
   ]
  }
  MaterialExpressionLinearInterpolate_14 {
   class=MaterialExpressionLinearInterpolate
   inputs [
    A {from=MaterialExpressionVectorParameter_2, output=RGB}
    B {from=MaterialExpressionLinearInterpolate_12}
    Alpha {from=MaterialExpressionNamedRerouteUsage_42}
   ]
  }
  MaterialExpressionLinearInterpolate_15 {
   class=MaterialExpressionLinearInterpolate
   inputs [
    A {from=MaterialExpressionNamedRerouteDeclaration_26}
    B {from=MaterialExpressionNamedRerouteUsage_34}
    Alpha {from=MaterialExpressionNamedRerouteUsage_33}
   ]
  }
  MaterialExpressionLinearInterpolate_16 {
   class=MaterialExpressionLinearInterpolate
   inputs [
    A {from=MaterialExpressionNamedRerouteDeclaration_25}
    B {from=MaterialExpressionNamedRerouteUsage_41}
    Alpha {from=MaterialExpressionNamedRerouteUsage_35}
   ]
  }
  MaterialExpressionLinearInterpolate_19 {
   class=MaterialExpressionLinearInterpolate
   inputs [
    A {from=""}
    B {from=""}
    Alpha {from=MaterialExpressionNamedRerouteDeclaration_28}
   ]
   properties {constA=1, constB=5}
  }
  MaterialExpressionLinearInterpolate_2 {
   class=MaterialExpressionLinearInterpolate
   inputs [
    A {from=MaterialExpressionScalarParameter_11}
    B {from=MaterialExpressionScalarParameter_6}
    Alpha {from=MaterialExpressionSaturate_10}
   ]
   properties {constB=0.4}
  }
  MaterialExpressionLinearInterpolate_20 {
   class=MaterialExpressionLinearInterpolate
   inputs [
    A {from=MaterialExpressionVectorParameter_2, output=RGB}
    B {from=MaterialExpressionLinearInterpolate_7}
    Alpha {from=MaterialExpressionNamedRerouteUsage_42}
   ]
  }
  MaterialExpressionLinearInterpolate_21 {
   class=MaterialExpressionLinearInterpolate
   inputs [
    A {from=MaterialExpressionScalarParameter_40}
    B {from=MaterialExpressionConvert_0}
    Alpha {from=MaterialExpressionNamedRerouteDeclaration_28}
   ]
   properties {constB=0}
  }
  MaterialExpressionLinearInterpolate_22 {
   class=MaterialExpressionLinearInterpolate
   inputs [
    A {from=MaterialExpressionScalarParameter_43}
    B {from=MaterialExpressionScalarParameter_41}
    Alpha {from=MaterialExpressionNamedRerouteUsage_40}
   ]
  }
  MaterialExpressionLinearInterpolate_23 {
   class=MaterialExpressionLinearInterpolate
   inputs [
    A {from=MaterialExpressionOneMinus_10}
    B {from=MaterialExpressionSaturate_22}
    Alpha {from=MaterialExpressionNamedRerouteUsage_45}
   ]
  }
  MaterialExpressionLinearInterpolate_25 {
   class=MaterialExpressionLinearInterpolate
   inputs [
    A {from=MaterialExpressionConstant_17}
    B {from=MaterialExpressionConstant_18}
    Alpha {from=MaterialExpressionNamedRerouteUsage_50}
   ]
   properties {constB=2}
  }
  MaterialExpressionLinearInterpolate_3 {
   class=MaterialExpressionLinearInterpolate
   inputs [
    A {from=MaterialExpressionScalarParameter_11}
    B {from=MaterialExpressionScalarParameter_6}
    Alpha {from=MaterialExpressionNamedRerouteDeclaration_14}
   ]
   properties {constB=0.6}
  }
  MaterialExpressionLinearInterpolate_4 {
   class=MaterialExpressionLinearInterpolate
   inputs [
    A {from=MaterialExpressionScalarParameter_11}
    B {from=MaterialExpressionScalarParameter_6}
    Alpha {from=MaterialExpressionSaturate_12}
   ]
   properties {constB=0.4}
  }
  MaterialExpressionLinearInterpolate_6 {
   class=MaterialExpressionLinearInterpolate
   inputs [
    A {from=MaterialExpressionVectorParameter_3, output=RGB}
    B {from=MaterialExpressionVectorParameter_4, output=RGB}
    Alpha {from=MaterialExpressionTextureSample_0, output=R}
   ]
  }
  MaterialExpressionLinearInterpolate_7 {
   class=MaterialExpressionLinearInterpolate
   inputs [
    A {from=MaterialExpressionVectorParameter_1, output=RGB}
    B {from=MaterialExpressionLinearInterpolate_6}
    Alpha {from=MaterialExpressionNamedRerouteUsage_44}
   ]
  }
  MaterialExpressionMaterialFunctionCall_0 {
   class=MaterialExpressionMaterialFunctionCall
   inputs [
    UVs {from=MaterialExpressionAdd_5}
    Texture Scale {from=MaterialExpressionScalarParameter_0}
   ]
   properties {materialFunction=&ScaleUVsByCenter}
  }
  MaterialExpressionMaterialFunctionCall_10 {
   class=MaterialExpressionMaterialFunctionCall
   inputs [Float2 {from=MaterialExpressionMaterialFunctionCall_0, output=UVs}]
   properties {materialFunction=&BreakOutFloat2Components}
  }
  MaterialExpressionMaterialFunctionCall_11 {
   class=MaterialExpressionMaterialFunctionCall
   inputs [
    X {from=MaterialExpressionAdd_22}
    Y {from=MaterialExpressionMaterialFunctionCall_10, output=G}
   ]
   properties {materialFunction=&MakeFloat2}
  }
  MaterialExpressionMaterialFunctionCall_14 {
   class=MaterialExpressionMaterialFunctionCall
   inputs [Float2 {from=MaterialExpressionMaterialFunctionCall_6, output=UVs}]
   properties {materialFunction=&BreakOutFloat2Components}
  }
  MaterialExpressionMaterialFunctionCall_16 {
   class=MaterialExpressionMaterialFunctionCall
   inputs [Float2 {from=MaterialExpressionNamedRerouteUsage_2}]
   properties {materialFunction=&BreakOutFloat2Components}
  }
  MaterialExpressionMaterialFunctionCall_17 {
   class=MaterialExpressionMaterialFunctionCall
   inputs [
    X {from=MaterialExpressionMaterialFunctionCall_16, output=R}
    Y {from=MaterialExpressionAdd_17}
   ]
   properties {materialFunction=&MakeFloat2}
  }
  MaterialExpressionMaterialFunctionCall_18 {
   class=MaterialExpressionMaterialFunctionCall
   inputs [
    X {from=MaterialExpressionAdd_18}
    Y {from=MaterialExpressionMaterialFunctionCall_14, output=G}
   ]
   properties {materialFunction=&MakeFloat2}
  }
  MaterialExpressionMaterialFunctionCall_2 {
   class=MaterialExpressionMaterialFunctionCall
   inputs [Float2 {from=MaterialExpressionFloor_0}]
   properties {materialFunction=&BreakOutFloat2Components}
  }
  MaterialExpressionMaterialFunctionCall_4 {
   class=MaterialExpressionMaterialFunctionCall
   inputs [Float2 {from=MaterialExpressionNamedRerouteUsage_22}]
   properties {materialFunction=&BreakOutFloat2Components}
  }
  MaterialExpressionMaterialFunctionCall_5 {
   class=MaterialExpressionMaterialFunctionCall
   inputs [Float2 {from=MaterialExpressionAdd_25}]
   properties {materialFunction=&BreakOutFloat2Components}
  }
  MaterialExpressionMaterialFunctionCall_6 {
   class=MaterialExpressionMaterialFunctionCall
   inputs [
    UVs {from=MaterialExpressionNamedRerouteDeclaration_30}
    Texture Scale {from=MaterialExpressionMultiply_37}
   ]
   properties {materialFunction=&ScaleUVsByCenter}
  }
  MaterialExpressionMaterialFunctionCall_7 {
   class=MaterialExpressionMaterialFunctionCall
   inputs [Float2 {from=MaterialExpressionMultiply_3}]
   properties {materialFunction=&BreakOutFloat2Components}
  }
  MaterialExpressionMax_11 {
   class=MaterialExpressionMax
   inputs [
    A {from=MaterialExpressionAdd_15}
    B {from=MaterialExpressionMaterialFunctionCall_2, output=G}
   ]
   properties {constB=0}
  }
  MaterialExpressionMax_12 {
   class=MaterialExpressionMax
   inputs [
    A {from=MaterialExpressionMaterialFunctionCall_7, output=R}
    B {from=MaterialExpressionMaterialFunctionCall_7, output=R}
   ]
   properties {constB=0}
  }
  MaterialExpressionMax_13 {
   class=MaterialExpressionMax
   inputs [
    A {from=MaterialExpressionMax_11}
    B {from=MaterialExpressionMax_12}
   ]
   properties {constB=0}
  }
  MaterialExpressionMax_16 {
   class=MaterialExpressionMax
   inputs [
    A {from=MaterialExpressionMin_2}
    B {from=MaterialExpressionMin_1}
   ]
  }
  MaterialExpressionMin_0 {
   class=MaterialExpressionMin
   inputs [
    A {from=MaterialExpressionCeil_1, output=""}
    B {from=MaterialExpressionNamedRerouteUsage_5, output=""}
   ]
   properties {}
  }
  MaterialExpressionMin_1 {
   class=MaterialExpressionMin
   inputs [
    A {from=MaterialExpressionMin_4, output=""}
    B {from=MaterialExpressionNamedRerouteUsage_28, output=""}
   ]
   properties {}
  }
  MaterialExpressionMin_2 {
   class=MaterialExpressionMin
   inputs [
    A {from=MaterialExpressionMin_3, output=""}
    B {from=MaterialExpressionNamedRerouteUsage_27, output=""}
   ]
   properties {}
  }
  MaterialExpressionMin_3 {
   class=MaterialExpressionMin
   inputs [
    A {from=MaterialExpressionNamedRerouteUsage_38, output=""}
    B {from=MaterialExpressionNamedRerouteUsage_36, output=""}
   ]
   properties {}
  }
  MaterialExpressionMin_4 {
   class=MaterialExpressionMin
   inputs [
    A {from=MaterialExpressionNamedRerouteUsage_39, output=""}
    B {from=MaterialExpressionNamedRerouteUsage_37, output=""}
   ]
   properties {}
  }
  MaterialExpressionMultiply_11 {
   class=MaterialExpressionMultiply
   inputs [
    A {from=""}
    B {from=""}
   ]
  }
  MaterialExpressionMultiply_13 {
   class=MaterialExpressionMultiply
   inputs [
    A {from=MaterialExpressionScalarParameter_27}
    B {from=""}
   ]
   properties {constB=-1}
  }
  MaterialExpressionMultiply_16 {
   class=MaterialExpressionMultiply
   inputs [
    A {from=MaterialExpressionScalarParameter_23}
    B {from=""}
   ]
   properties {constB=-1}
  }
  MaterialExpressionMultiply_18 {
   class=MaterialExpressionMultiply
   inputs [
    A {from=MaterialExpressionMaterialFunctionCall_5, output=G}
    B {from=""}
   ]
   properties {constB=-1}
  }
  MaterialExpressionMultiply_19 {
   class=MaterialExpressionMultiply
   inputs [
    A {from=MaterialExpressionAdd_24}
    B {from=""}
   ]
   properties {constB=-1}
  }
  MaterialExpressionMultiply_20 {
   class=MaterialExpressionMultiply
   inputs [
    A {from=MaterialExpressionMin_0}
    B {from=MaterialExpressionFunctionInput_0}
   ]
  }
  MaterialExpressionMultiply_27 {
   class=MaterialExpressionMultiply
   inputs [
    A {from=MaterialExpressionNamedRerouteUsage_3}
    B {from=MaterialExpressionConstant2Vector_0}
   ]
  }
  MaterialExpressionMultiply_3 {
   class=MaterialExpressionMultiply
   inputs [
    A {from=MaterialExpressionFloor_0}
    B {from=""}
   ]
   properties {constB=-1}
  }
  MaterialExpressionMultiply_35 {
   class=MaterialExpressionMultiply
   inputs [
    A {from=MaterialExpressionSubtract_0}
    B {from=""}
   ]
   properties {constB=2}
  }
  MaterialExpressionMultiply_37 {
   class=MaterialExpressionMultiply
   inputs [
    A {from=MaterialExpressionAppendVector_2}
    B {from=MaterialExpressionLinearInterpolate_21}
   ]
  }
  MaterialExpressionMultiply_38 {
   class=MaterialExpressionMultiply
   inputs [
    A {from=MaterialExpressionSubtract_3}
    B {from=""}
   ]
   properties {constB=2}
  }
  MaterialExpressionMultiply_40 {
   class=MaterialExpressionMultiply
   inputs [
    A {from=MaterialExpressionClamp_2}
    B {from=MaterialExpressionScalarParameter_21}
   ]
  }
  MaterialExpressionMultiply_41 {
   class=MaterialExpressionMultiply
   inputs [
    A {from=MaterialExpressionClamp_3}
    B {from=MaterialExpressionScalarParameter_18}
   ]
  }
  MaterialExpressionMultiply_5 {
   class=MaterialExpressionMultiply
   inputs [
    A {from=MaterialExpressionAdd_16}
    B {from=MaterialExpressionScalarParameter_7}
   ]
  }
  MaterialExpressionMultiply_6 {
   class=MaterialExpressionMultiply
   inputs [
    A {from=MaterialExpressionOneMinus_2}
    B {from=MaterialExpressionNamedRerouteDeclaration_2}
   ]
   properties {desc=RightEye}
  }
  MaterialExpressionNamedRerouteDeclaration_0 {
   class=MaterialExpressionNamedRerouteDeclaration
   inputs [None {from=MaterialExpressionStaticSwitch_2}]
   properties {name=EyeUVs, nodeColor=<rgba 0.2188 0 1 1>}
  }
  MaterialExpressionNamedRerouteDeclaration_1 {
   class=MaterialExpressionNamedRerouteDeclaration
   inputs [None {from=MaterialExpressionMaterialFunctionCall_18, output=Result}]
   properties {name=EyelidCoords, nodeColor=<rgba 0 1 0.9531 1>}
  }
  MaterialExpressionNamedRerouteDeclaration_11 {
   class=MaterialExpressionNamedRerouteDeclaration
   inputs [None {from=MaterialExpressionStaticSwitchParameter_3}]
   properties {name=EyeColor_Lit, nodeColor=<rgba 1 0.5322 0.1287 1>}
  }
  MaterialExpressionNamedRerouteDeclaration_12 {
   class=MaterialExpressionNamedRerouteDeclaration
   inputs [None {from=MaterialExpressionOneMinus_1}]
   properties {name=Mask_EyesUDIM, nodeColor=<rgba 1 0.2578 0 1>}
  }
  MaterialExpressionNamedRerouteDeclaration_14 {
   class=MaterialExpressionNamedRerouteDeclaration
   inputs [None {from=MaterialExpressionSaturate_9}]
   properties {name=R_upperLid_close_pose, nodeColor=<rgba 0.125 1 0 1>}
  }
  MaterialExpressionNamedRerouteDeclaration_15 {
   class=MaterialExpressionNamedRerouteDeclaration
   inputs [None {from=MaterialExpressionSaturate_7}]
   properties {name=EyelidMask, nodeColor=<rgba 1 0.5391 0 1>}
  }
  MaterialExpressionNamedRerouteDeclaration_16 {
   class=MaterialExpressionNamedRerouteDeclaration
   inputs [None {from=MaterialExpressionSaturate_11}]
   properties {name=L_upperLid_close_pose, nodeColor=<rgba 0.125 1 0 1>}
  }
  MaterialExpressionNamedRerouteDeclaration_18 {
   class=MaterialExpressionNamedRerouteDeclaration
   inputs [None {from=MaterialExpressionAbs_4}]
   properties {name=UR_Grad, nodeColor=<rgba 0.15 1 0 1>}
  }
  MaterialExpressionNamedRerouteDeclaration_2 {
   class=MaterialExpressionNamedRerouteDeclaration
   inputs [None {from=MaterialExpressionMultiply_20}]
   properties {name=Mask_Eyes, nodeColor=<rgba 1 0 0.8906 1>}
  }
  MaterialExpressionNamedRerouteDeclaration_20 {
   class=MaterialExpressionNamedRerouteDeclaration
   inputs [None {from=MaterialExpressionAbs_6}]
   properties {name=LR_Grad, nodeColor=<rgba 0 1 0.0156 1>}
  }
  MaterialExpressionNamedRerouteDeclaration_21 {
   class=MaterialExpressionNamedRerouteDeclaration
   inputs [None {from=MaterialExpressionAbs_8}]
   properties {name=UL_Grad, nodeColor=<rgba 1 0.5333 0 1>}
  }
  MaterialExpressionNamedRerouteDeclaration_22 {
   class=MaterialExpressionNamedRerouteDeclaration
   inputs [None {from=MaterialExpressionAbs_9}]
   properties {name=LL_Grad, nodeColor=<rgba 0.8952 1 0 1>}
  }
  MaterialExpressionNamedRerouteDeclaration_23 {
   class=MaterialExpressionNamedRerouteDeclaration
   inputs [None {from=MaterialExpressionSmoothStep_1}]
   properties {name=DF_Lids, nodeColor=<rgba 1 0.3047 0 1>}
  }
  MaterialExpressionNamedRerouteDeclaration_24 {
   class=MaterialExpressionNamedRerouteDeclaration
   inputs [None {from=MaterialExpressionReroute_0}]
   properties {name=PupilCoords, nodeColor=<rgba 0 1 0.6016 1>}
  }
  MaterialExpressionNamedRerouteDeclaration_25 {
   class=MaterialExpressionNamedRerouteDeclaration
   inputs [None {from=MaterialExpressionFunctionInput_2}]
   properties {name=InColorLit, nodeColor=<rgba 1 0 0.0234 1>}
  }
  MaterialExpressionNamedRerouteDeclaration_26 {
   class=MaterialExpressionNamedRerouteDeclaration
   inputs [None {from=MaterialExpressionFunctionInput_3}]
   properties {name=InColorShaded, nodeColor=<rgba 0 0.3437 1 1>}
  }
  MaterialExpressionNamedRerouteDeclaration_27 {
   class=MaterialExpressionNamedRerouteDeclaration
   inputs [None {from=MaterialExpressionStaticSwitchParameter_6}]
   properties {name=EyeColor_Shaded, nodeColor=<rgba 1 0.5322 0.1287 1>}
  }
  MaterialExpressionNamedRerouteDeclaration_28 {
   class=MaterialExpressionNamedRerouteDeclaration
   inputs [None {from=MaterialExpressionLinearInterpolate_22}]
   properties {name=Lid_close_pose, nodeColor=<rgba 0.125 1 0 1>}
  }
  MaterialExpressionNamedRerouteDeclaration_29 {
   class=MaterialExpressionNamedRerouteDeclaration
   inputs [None {from=MaterialExpressionLinearInterpolate_23}]
   properties {name=DEBUG, nodeColor=<rgba 1 0.3047 0 1>}
  }
  MaterialExpressionNamedRerouteDeclaration_3 {
   class=MaterialExpressionNamedRerouteDeclaration
   inputs [None {from=MaterialExpressionMultiply_6}]
   properties {name=Mask_EyeRight, nodeColor=<rgba 1 0 0.0152 1>}
  }
  MaterialExpressionNamedRerouteDeclaration_30 {
   class=MaterialExpressionNamedRerouteDeclaration
   inputs [None {from=MaterialExpressionAdd_5}]
   properties {name=EyelidCoords_pre, nodeColor=<rgba 0.6875 0 1 1>}
  }
  MaterialExpressionNamedRerouteDeclaration_4 {
   class=MaterialExpressionNamedRerouteDeclaration
   inputs [None {from=MaterialExpressionFloor_1}]
   properties {name=Mask_EyeLeft, nodeColor=<rgba 0.0293 0.3125 0.0353 1>}
  }
  MaterialExpressionNamedRerouteUsage_0 {class=MaterialExpressionNamedRerouteUsage}
  MaterialExpressionNamedRerouteUsage_2 {class=MaterialExpressionNamedRerouteUsage}
  MaterialExpressionNamedRerouteUsage_21 {class=MaterialExpressionNamedRerouteUsage}
  MaterialExpressionNamedRerouteUsage_22 {class=MaterialExpressionNamedRerouteUsage}
  MaterialExpressionNamedRerouteUsage_24 {class=MaterialExpressionNamedRerouteUsage}
  MaterialExpressionNamedRerouteUsage_27 {class=MaterialExpressionNamedRerouteUsage}
  MaterialExpressionNamedRerouteUsage_28 {class=MaterialExpressionNamedRerouteUsage}
  MaterialExpressionNamedRerouteUsage_3 {class=MaterialExpressionNamedRerouteUsage}
  MaterialExpressionNamedRerouteUsage_31 {class=MaterialExpressionNamedRerouteUsage}
  MaterialExpressionNamedRerouteUsage_33 {class=MaterialExpressionNamedRerouteUsage}
  MaterialExpressionNamedRerouteUsage_34 {class=MaterialExpressionNamedRerouteUsage}
  MaterialExpressionNamedRerouteUsage_35 {class=MaterialExpressionNamedRerouteUsage}
  MaterialExpressionNamedRerouteUsage_36 {class=MaterialExpressionNamedRerouteUsage}
  MaterialExpressionNamedRerouteUsage_37 {class=MaterialExpressionNamedRerouteUsage}
  MaterialExpressionNamedRerouteUsage_38 {class=MaterialExpressionNamedRerouteUsage}
  MaterialExpressionNamedRerouteUsage_39 {class=MaterialExpressionNamedRerouteUsage}
  MaterialExpressionNamedRerouteUsage_40 {class=MaterialExpressionNamedRerouteUsage}
  MaterialExpressionNamedRerouteUsage_41 {class=MaterialExpressionNamedRerouteUsage}
  MaterialExpressionNamedRerouteUsage_42 {class=MaterialExpressionNamedRerouteUsage}
  MaterialExpressionNamedRerouteUsage_43 {class=MaterialExpressionNamedRerouteUsage}
  MaterialExpressionNamedRerouteUsage_44 {class=MaterialExpressionNamedRerouteUsage}
  MaterialExpressionNamedRerouteUsage_45 {class=MaterialExpressionNamedRerouteUsage}
  MaterialExpressionNamedRerouteUsage_48 {class=MaterialExpressionNamedRerouteUsage}
  MaterialExpressionNamedRerouteUsage_5 {class=MaterialExpressionNamedRerouteUsage}
  MaterialExpressionNamedRerouteUsage_50 {class=MaterialExpressionNamedRerouteUsage}
  MaterialExpressionNamedRerouteUsage_51 {class=MaterialExpressionNamedRerouteUsage}
  MaterialExpressionNamedRerouteUsage_52 {class=MaterialExpressionNamedRerouteUsage}
  MaterialExpressionOneMinus_0 {class=MaterialExpressionOneMinus, inputs [None {from=MaterialExpressionScalarParameter_38}]}
  MaterialExpressionOneMinus_1 {class=MaterialExpressionOneMinus, inputs [None {from=MaterialExpressionMax_13}]}
  MaterialExpressionOneMinus_10 {class=MaterialExpressionOneMinus, inputs [None {from=MaterialExpressionSphereMask_3}]}
  MaterialExpressionOneMinus_2 {class=MaterialExpressionOneMinus, inputs [None {from=MaterialExpressionFloor_1}]}
  MaterialExpressionPower_0 {
   class=MaterialExpressionPower
   inputs [
    Base {from=MaterialExpressionScalarParameter_1}
    Exp {from=MaterialExpressionScalarParameter_19}
   ]
  }
  MaterialExpressionPower_1 {
   class=MaterialExpressionPower
   inputs [
    Base {from=MaterialExpressionScalarParameter_1}
    Exp {from=MaterialExpressionScalarParameter_19}
   ]
  }
  MaterialExpressionPower_2 {
   class=MaterialExpressionPower
   inputs [
    Base {from=MaterialExpressionScalarParameter_9}
    Exp {from=MaterialExpressionScalarParameter_19}
   ]
  }
  MaterialExpressionPower_3 {
   class=MaterialExpressionPower
   inputs [
    Base {from=MaterialExpressionScalarParameter_9}
    Exp {from=MaterialExpressionScalarParameter_19}
   ]
  }
  MaterialExpressionReroute_0 {class=MaterialExpressionReroute, inputs [None {from=MaterialExpressionMaterialFunctionCall_11, output=Result}]}
  MaterialExpressionReroute_1 {class=MaterialExpressionReroute, inputs [None {from=MaterialExpressionMaterialFunctionCall_5, output=G}]}
  MaterialExpressionReroute_2 {class=MaterialExpressionReroute, inputs [None {from=MaterialExpressionMaterialFunctionCall_5, output=G}]}
  MaterialExpressionReroute_3 {class=MaterialExpressionReroute, inputs [None {from=MaterialExpressionStaticSwitch_2}]}
  MaterialExpressionSaturate_10 {class=MaterialExpressionSaturate, inputs [None {from=MaterialExpressionPower_1}]}
  MaterialExpressionSaturate_11 {class=MaterialExpressionSaturate, inputs [None {from=MaterialExpressionPower_2}]}
  MaterialExpressionSaturate_12 {class=MaterialExpressionSaturate, inputs [None {from=MaterialExpressionPower_3}]}
  MaterialExpressionSaturate_22 {class=MaterialExpressionSaturate, inputs [None {from=MaterialExpressionMax_16}]}
  MaterialExpressionSaturate_7 {class=MaterialExpressionSaturate, inputs [None {from=MaterialExpressionSmoothStep_0}]}
  MaterialExpressionSaturate_9 {class=MaterialExpressionSaturate, inputs [None {from=MaterialExpressionPower_0}]}
  MaterialExpressionScalarParameter_0 {
   class=MaterialExpressionScalarParameter
   properties {defaultValue=0.5, group=01_EyeAnimation_Parameters, parameterName=PupilMap_Scale, sliderMax=2, sortPriority=5}
  }
  MaterialExpressionScalarParameter_1 {
   class=MaterialExpressionScalarParameter
   properties {defaultValue=1, group=01_EyeAnimation_Poses, parameterName=R_blink_pose, sliderMax=1, sortPriority=2}
  }
  MaterialExpressionScalarParameter_11 {
   class=MaterialExpressionScalarParameter
   properties {defaultValue=0.3288, group=01_EyeAnimation_Parameters, parameterName=Lid_Min, sliderMax=1, sortPriority=28}
  }
  MaterialExpressionScalarParameter_18 {
   class=MaterialExpressionScalarParameter
   properties {defaultValue=0.2, group=01_EyeAnimation_Parameters, parameterName=EyeVertical_Threshold, sliderMax=1}
  }
  MaterialExpressionScalarParameter_19 {
   class=MaterialExpressionScalarParameter
   properties {defaultValue=1, group=01_EyeAnimation_Parameters, parameterName=AnimationThreshold_Eyes, sliderMax=8, sortPriority=30}
  }
  MaterialExpressionScalarParameter_20 {class=MaterialExpressionScalarParameter, properties {group=01_EyeAnimation_Poses, parameterName=look_up_pose, sliderMax=1, sortPriority=12}}
  MaterialExpressionScalarParameter_21 {
   class=MaterialExpressionScalarParameter
   properties {defaultValue=0.2, group=01_EyeAnimation_Parameters, parameterName=EyeHorizontal_Threshold, sliderMax=1}
  }
  MaterialExpressionScalarParameter_23 {
   class=MaterialExpressionScalarParameter
   properties {group=01_EyeAnimation_Poses, parameterName=look_right_pose, sliderMax=1, sortPriority=10}
  }
  MaterialExpressionScalarParameter_26 {
   class=MaterialExpressionScalarParameter
   properties {group=01_EyeAnimation_Poses, parameterName=look_left_pose, sliderMax=1, sortPriority=11}
  }
  MaterialExpressionScalarParameter_27 {
   class=MaterialExpressionScalarParameter
   properties {group=01_EyeAnimation_Poses, parameterName=look_down_pose, sliderMax=1, sortPriority=13}
  }
  MaterialExpressionScalarParameter_37 {
   class=MaterialExpressionScalarParameter
   properties {defaultValue=0.9, group=01_EyeAnimation_Parameters, parameterName=Lid_DF_LineHardness, sliderMax=1}
  }
  MaterialExpressionScalarParameter_38 {
   class=MaterialExpressionScalarParameter
   properties {defaultValue=0.05, group=01_EyeAnimation_Parameters, parameterName=Lid_DF_LineThickness, sliderMax=1}
  }
  MaterialExpressionScalarParameter_40 {class=MaterialExpressionScalarParameter, properties {defaultValue=2, group=01_EyeAnimation_Parameters, parameterName=EyelidScale}}
  MaterialExpressionScalarParameter_41 {
   class=MaterialExpressionScalarParameter
   properties {defaultValue=0.2, group=01_EyeAnimation_Poses, parameterName=R_blink_pose, sliderMax=1, sortPriority=2}
  }
  MaterialExpressionScalarParameter_43 {class=MaterialExpressionScalarParameter, properties {group=01_EyeAnimation_Poses, parameterName=L_blink_pose, sliderMax=1, sortPriority=1}}
  MaterialExpressionScalarParameter_6 {
   class=MaterialExpressionScalarParameter
   properties {defaultValue=0.5, group=01_EyeAnimation_Parameters, parameterName=Lid_Max, sliderMax=1, sortPriority=29}
  }
  MaterialExpressionScalarParameter_7 {
   class=MaterialExpressionScalarParameter
   properties {defaultValue=2, desc=horizontal UV ratio, group=01_EyeAnimation_Parameters, parameterName=PupilMap_UV_Ratio, sliderMax=2, sortPriority=5}
  }
  MaterialExpressionScalarParameter_9 {class=MaterialExpressionScalarParameter, properties {group=01_EyeAnimation_Poses, parameterName=L_blink_pose, sliderMax=1, sortPriority=1}}
  MaterialExpressionSmoothStep_0 {
   class=MaterialExpressionSmoothStep
   inputs [
    Min {from="", output=""}
    Max {from="", output=""}
    Value {from=MaterialExpressionTextureSample_2, output=G}
   ]
   properties {constMax=0.25, constMin=0.2}
  }
  MaterialExpressionSmoothStep_1 {
   class=MaterialExpressionSmoothStep
   inputs [
    Min {from="", output=""}
    Max {from="", output=""}
    Value {from=MaterialExpressionSphereMask_1, output=""}
   ]
   properties {constMax=0.6, constMin=0.5}
  }
  MaterialExpressionSphereMask_1 {
   class=MaterialExpressionSphereMask
   inputs [
    A {from=MaterialExpressionLinearInterpolate_23, output=""}
    B {from=MaterialExpressionConstant_7, output=""}
    Radius {from=MaterialExpressionOneMinus_0, output=""}
    Hardness {from=MaterialExpressionScalarParameter_37, output=""}
   ]
   properties {attenuationRadius=0.6, hardnessPercent=95}
  }
  MaterialExpressionSphereMask_3 {
   class=MaterialExpressionSphereMask
   inputs [
    A {from=MaterialExpressionDistance_2, output=""}
    B {from=MaterialExpressionConstant_14, output=""}
    Radius {from=MaterialExpressionLinearInterpolate_25, output=""}
    Hardness {from="", output=""}
   ]
   properties {hardnessPercent=1}
  }
  MaterialExpressionStaticBool_0 {class=MaterialExpressionStaticBool, properties {value=true}}
  MaterialExpressionStaticSwitchParameter_3 {
   class=MaterialExpressionStaticSwitchParameter
   inputs [
    True {from=MaterialExpressionLinearInterpolate_7, output=""}
    False {from=MaterialExpressionLinearInterpolate_20, output=""}
   ]
   properties {group=01_EyeAnimation_Parameters, parameterName=UseOverlapOutlines, sortPriority=0}
  }
  MaterialExpressionStaticSwitchParameter_4 {
   class=MaterialExpressionStaticSwitchParameter
   inputs [
    True {from=MaterialExpressionLinearInterpolate_15, output=""}
    False {from=MaterialExpressionNamedRerouteDeclaration_26, output=""}
   ]
   properties {group=01_EyeAnimation_Parameters, parameterName=UseEyeAnimation, sortPriority=-1}
  }
  MaterialExpressionStaticSwitchParameter_5 {
   class=MaterialExpressionStaticSwitchParameter
   inputs [
    True {from=MaterialExpressionLinearInterpolate_16, output=""}
    False {from=MaterialExpressionNamedRerouteDeclaration_25, output=""}
   ]
   properties {group=01_EyeAnimation_Parameters, parameterName=UseEyeAnimation, sortPriority=-1}
  }
  MaterialExpressionStaticSwitchParameter_6 {
   class=MaterialExpressionStaticSwitchParameter
   inputs [
    True {from=MaterialExpressionLinearInterpolate_12, output=""}
    False {from=MaterialExpressionLinearInterpolate_14, output=""}
   ]
   properties {group=01_EyeAnimation_Parameters, parameterName=UseOverlapOutlines, sortPriority=0}
  }
  MaterialExpressionStaticSwitch_2 {
   class=MaterialExpressionStaticSwitch
   inputs [
    True {from=MaterialExpressionTextureCoordinate_5, output=""}
    False {from=MaterialExpressionTextureCoordinate_4, output=""}
    Value {from=MaterialExpressionFunctionInput_1, output=""}
   ]
   properties {}
  }
  MaterialExpressionSubtract_0 {
   class=MaterialExpressionSubtract
   inputs [
    A {from=MaterialExpressionMaterialFunctionCall_10, output=R}
    B {from=""}
   ]
   properties {constB=0.5}
  }
  MaterialExpressionSubtract_3 {
   class=MaterialExpressionSubtract
   inputs [
    A {from=MaterialExpressionMaterialFunctionCall_14, output=R}
    B {from=""}
   ]
   properties {constB=0.5}
  }
  MaterialExpressionTextureCoordinate_4 {class=MaterialExpressionTextureCoordinate}
  MaterialExpressionTextureCoordinate_5 {class=MaterialExpressionTextureCoordinate, properties {coordinateIndex=1}}
  MaterialExpressionTextureSample_0 {
   class=MaterialExpressionTextureSample
   inputs [
    UVs {from=MaterialExpressionNamedRerouteUsage_43}
    Tex {from=""}
    Apply View MipBias {from=""}
   ]
   properties {
    samplerSource=SSM_Clamp_WorldGroupSettings
    samplerType=SAMPLERTYPE_Masks
    texture=/BRCosmetics/Accessories/FORT_Backpacks/Backpack_M_MED_BiomeHose/Textures/T_BiomeHose_Backpack_Eye
   }
  }
  MaterialExpressionTextureSample_2 {
   class=MaterialExpressionTextureSample
   inputs [
    UVs {from=MaterialExpressionNamedRerouteUsage_31}
    Tex {from=""}
    Apply View MipBias {from=""}
   ]
   properties {
    samplerSource=SSM_Clamp_WorldGroupSettings
    samplerType=SAMPLERTYPE_Masks
    texture=&T_BiomeHose_Backpack_Eye
   }
  }
  MaterialExpressionVectorParameter_0 {
   class=MaterialExpressionVectorParameter
   properties {group=01_EyeAnimation_Parameters, parameterName=EyeLid_Color_Shaded, sortPriority=3, defaultValue=<rgba 0.5972 0.3813 0.0343 1>}
  }
  MaterialExpressionVectorParameter_1 {
   class=MaterialExpressionVectorParameter
   properties {group=01_EyeAnimation_Parameters, parameterName=EyeLid_Color_Lit, sortPriority=3, defaultValue=<rgba 1 0.6445 0.0452 1>}
  }
  MaterialExpressionVectorParameter_2 {
   class=MaterialExpressionVectorParameter
   properties {group=01_EyeAnimation_Parameters, parameterName=EyeLid_Color_DF_Line, sortPriority=3, defaultValue=<rgba 0 0 0 1>}
  }
  MaterialExpressionVectorParameter_3 {
   class=MaterialExpressionVectorParameter
   properties {group=01_EyeAnimation_Parameters, parameterName=Eyeball_Color, sortPriority=2, defaultValue=<rgba 1 1 1 0>}
  }
  MaterialExpressionVectorParameter_4 {class=MaterialExpressionVectorParameter, properties {group=01_EyeAnimation_Parameters, parameterName=Pupil_Color, sortPriority=2}}
 }}
