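# Animation Blueprint for the BiomeHose backpack cosmetic, animating idle, skydive and chest-open reactions.
# In the front end, a state machine replays the chest-open animation on a delay timer; otherwise another handles skydiving with random loops and the ChestOpened input event.
# The pose then gets an additive pack-bone offset driven by spine bend, followed by spring AnimDynamics on the moustache bones.
# The chain-bone dynamics nodes are not connected to the output.
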AnimBlueprint /BRCosmetics/Accessories/FORT_Backpacks/Backpack_M_MED_BiomeHose/Meshes/M_MED_BiomeHose_Pack_AnimBP
target_skeleton=&/M_MED_BiomeHose_Pack_Skeleton
graphs [
 AnimGraph {nodes [
   n1 = Anim Dynamics - Bone: chain1(Alpha=1, ExternalForce="0,0,0") {
    angularDampingOverride=0.99
    angularSpringConstant=60
    bAngularSpring=true
    bOverrideAngularDamping=true
    bUseGravityOverride=true
    linearDampingOverride=0.99
    linearSpringConstant=550
    boundBone {boneName=chain1}
    componentAppliedLinearAccClamp=<3000 3000 6000>
    componentLinearAccScale=<0.4 0.4 0.4>
    componentLinearVelScale=<0.1 0.1 0.1>
    gravityOverride {z=-980}
    physicsBodyDefinitions [{
      boundBone {boneName=chain1}
      boxExtents=<10 10 10>
      constraintSetup {angularLimitsMax=<0 0 55>, angularLimitsMin=<0 0 -55>, angularTarget=<1 0 0>}
      localJointOffset=<1 0 0>
     }]
   }
   Anim Dynamics - Bone: chain2(ComponentPose=@n1, Alpha=1, ExternalForce="0,0,0") {
    angularDampingOverride=0.99
    angularSpringConstant=550
    bOverrideAngularDamping=true
    bUseGravityOverride=true
    linearDampingOverride=0.99
    linearSpringConstant=550
    worldSpaceMinimumScale=0.1
    boundBone {boneName=chain2}
    componentAppliedLinearAccClamp=<3000 3000 6000>
    componentLinearAccScale=<0.3 0.3 1>
    componentLinearVelScale=<0.2 0.2 0.1>
    gravityOverride {z=-980}
    physicsBodyDefinitions [{
      boundBone {boneName=chain2}
      boxExtents=<10 10 10>
      constraintSetup {angularLimitsMax=<5 40 20>, angularLimitsMin=<-5 -80 -20>, angularTarget=<1 0 0>}
      localJointOffset=<1 0.1 0.1>
     }]
   }
   Anim Dynamics - Bone: moustache1_03_l(Alpha=1, ExternalForce="0,0,0") {
    angularDampingOverride=0.99
    angularSpringConstant=550
    bOverrideAngularDamping=true
    bUseGravityOverride=true
    linearDampingOverride=0.99
    linearSpringConstant=550
    worldSpaceMinimumScale=0.1
    boundBone {boneName=moustache1_03_l}
    componentAppliedLinearAccClamp=<3000 3000 6000>
    componentLinearAccScale=<0.3 0.3 1>
    componentLinearVelScale=<0.2 0.2 0.1>
    gravityOverride {z=-980}
    physicsBodyDefinitions [{
      boundBone {boneName=moustache1_03_l}
      boxExtents=<10 10 10>
      constraintSetup {angularLimitsMax=<0 10 10>, angularLimitsMin=<0 -10 -10>, angularTarget=<1 0 0>}
      localJointOffset=<1 0.1 0.1>
     }]
   }
   Input Pose()
   AnimGraphNode_StateMachine_1 = FE_Reactivity() {is=title AnimGraphNode_StateMachine}
   AnimGraphNode_StateMachine_2 = Idle and Reactive() {is=title AnimGraphNode_StateMachine}
   n2 = Get bIsInFrontEnd()
   n3 = "Transform (Modify) Bone - Bone: pack"(
    ComponentPose=Local To Component(LocalPose=Blend Poses by bool(BlendPose_0=@AnimGraphNode_StateMachine_1, BlendPose_1=@AnimGraphNode_StateMachine_2, bActiveValue=@n2) {blendType=Linear})
    Translation=Get backpack_translation()
    Rotation=Get backpack_rotate()
   ) {rotationMode=BMM_Additive, translationMode=BMM_Additive, alphaBoolBlend {blendOption=Linear}, boneToModify {boneName=pack}}
   n4 = Anim Dynamics - Bone: moustache1_01_r(ComponentPose=@n3, Alpha=1, ExternalForce="0,0,0") {
    angularDampingOverride=0.999
    angularSpringConstant=250
    bAngularSpring=true
    bOverrideAngularDamping=true
    bUseGravityOverride=true
    linearDampingOverride=0.99
    linearSpringConstant=550
    boundBone {boneName=moustache1_01_r}
    componentAppliedLinearAccClamp=<3000 3000 6000>
    componentLinearAccScale=<0.4 0.4 0.4>
    componentLinearVelScale=<0.1 0.1 0.1>
    gravityOverride {z=-980}
    physicsBodyDefinitions [{
      boundBone {boneName=moustache1_01_r}
      boxExtents=<10 10 10>
      constraintSetup {angularLimitsMax=<0 0 40>, angularLimitsMin=<0 0 -40>, angularTarget=<1 0 0>}
      localJointOffset=<1 0 0>
     }]
   }
   n5 = Anim Dynamics - Bone: moustache1_02_r(ComponentPose=@n4, Alpha=1, ExternalForce="0,0,0") {
    angularDampingOverride=0.999
    angularSpringConstant=250
    bAngularSpring=true
    bOverrideAngularDamping=true
    bUseGravityOverride=true
    linearDampingOverride=0.99
    linearSpringConstant=550
    worldSpaceMinimumScale=0.1
    boundBone {boneName=moustache1_02_r}
    componentAppliedLinearAccClamp=<3000 3000 6000>
    componentLinearAccScale=<0.3 0.3 1>
    componentLinearVelScale=<0.2 0.2 0.1>
    gravityOverride {z=-980}
    physicsBodyDefinitions [{
      boundBone {boneName=moustache1_02_r}
      boxExtents=<10 10 10>
      constraintSetup {angularLimitsMax=<0 40 20>, angularLimitsMin=<0 -40 -20>, angularTarget=<1 0 0>}
      localJointOffset=<1 0.1 0.1>
     }]
   }
   n6 = Anim Dynamics - Bone: moustache1_03_r(ComponentPose=@n5, Alpha=1, ExternalForce="0,0,0") {
    angularDampingOverride=0.999
    angularSpringConstant=250
    bAngularSpring=true
    bOverrideAngularDamping=true
    bUseGravityOverride=true
    linearDampingOverride=0.99
    linearSpringConstant=550
    worldSpaceMinimumScale=0.1
    boundBone {boneName=moustache1_03_r}
    componentAppliedLinearAccClamp=<3000 3000 6000>
    componentLinearAccScale=<0.3 0.3 1>
    componentLinearVelScale=<0.2 0.2 0.1>
    gravityOverride {z=-980}
    physicsBodyDefinitions [{
      boundBone {boneName=moustache1_03_r}
      boxExtents=<10 10 10>
      constraintSetup {angularLimitsMax=<0 20 20>, angularLimitsMin=<0 -20 -20>, angularTarget=<1 0 0>}
      localJointOffset=<1 0.1 0.1>
     }]
   }
   n7 = Anim Dynamics - Bone: moustache1_01_l(ComponentPose=@n6, Alpha=1, ExternalForce="0,0,0") {
    angularDampingOverride=0.999
    angularSpringConstant=250
    bAngularSpring=true
    bOverrideAngularDamping=true
    bUseGravityOverride=true
    linearDampingOverride=0.99
    linearSpringConstant=550
    boundBone {boneName=moustache1_01_l}
    componentAppliedLinearAccClamp=<3000 3000 6000>
    componentLinearAccScale=<0.4 0.4 0.4>
    componentLinearVelScale=<0.1 0.1 0.1>
    gravityOverride {z=-980}
    physicsBodyDefinitions [{
      boundBone {boneName=moustache1_01_l}
      boxExtents=<10 10 10>
      constraintSetup {angularLimitsMax=<0 0 40>, angularLimitsMin=<0 0 -40>, angularTarget=<1 0 0>}
      localJointOffset=<1 0 0>
     }]
   }
   n8 = Anim Dynamics - Bone: moustache1_02_l(ComponentPose=@n7, Alpha=1, ExternalForce="0,0,0") {
    angularDampingOverride=0.999
    angularSpringConstant=250
    bAngularSpring=true
    bOverrideAngularDamping=true
    bUseGravityOverride=true
    linearDampingOverride=0.99
    linearSpringConstant=550
    worldSpaceMinimumScale=0.1
    boundBone {boneName=moustache1_02_l}
    componentAppliedLinearAccClamp=<3000 3000 6000>
    componentLinearAccScale=<0.3 0.3 1>
    componentLinearVelScale=<0.2 0.2 0.1>
    gravityOverride {z=-980}
    physicsBodyDefinitions [{
      boundBone {boneName=moustache1_02_l}
      boxExtents=<10 10 10>
      constraintSetup {angularLimitsMax=<0 40 20>, angularLimitsMin=<0 -40 -20>, angularTarget=<1 0 0>}
      localJointOffset=<1 0.1 0.1>
     }]
   }
   n9 = Anim Dynamics - Bone: moustache1_03_l(ComponentPose=@n8, Alpha=1, ExternalForce="0,0,0") {
    angularDampingOverride=0.999
    angularSpringConstant=250
    bAngularSpring=true
    bOverrideAngularDamping=true
    bUseGravityOverride=true
    linearDampingOverride=0.99
    linearSpringConstant=550
    worldSpaceMinimumScale=0.1
    boundBone {boneName=moustache1_03_l}
    componentAppliedLinearAccClamp=<3000 3000 6000>
    componentLinearAccScale=<0.3 0.3 1>
    componentLinearVelScale=<0.2 0.2 0.1>
    gravityOverride {z=-980}
    physicsBodyDefinitions [{
      boundBone {boneName=moustache1_03_l}
      boxExtents=<10 10 10>
      constraintSetup {angularLimitsMax=<0 20 20>, angularLimitsMin=<0 -20 -20>, angularTarget=<1 0 0>}
      localJointOffset=<1 0.1 0.1>
     }]
   }
   n10 = Anim Dynamics - Bone: moustache2_01_r(ComponentPose=@n9, Alpha=1, ExternalForce="0,0,0") {
    angularDampingOverride=0.999
    angularSpringConstant=250
    bAngularSpring=true
    bOverrideAngularDamping=true
    bUseGravityOverride=true
    linearDampingOverride=0.99
    linearSpringConstant=550
    worldSpaceMinimumScale=0.1
    boundBone {boneName=moustache2_01_r}
    componentAppliedLinearAccClamp=<3000 3000 6000>
    componentLinearAccScale=<0.3 0.3 1>
    componentLinearVelScale=<0.2 0.2 0.1>
    gravityOverride {z=-980}
    physicsBodyDefinitions [{
      boundBone {boneName=moustache2_01_r}
      boxExtents=<10 10 10>
      constraintSetup {angularLimitsMax=<0 20 20>, angularLimitsMin=<0 -20 -20>, angularTarget=<1 0 0>}
      localJointOffset=<1 0.1 0.1>
     }]
   }
   n11 = Anim Dynamics - Bone: moustache2_02_r(ComponentPose=@n10, Alpha=1, ExternalForce="0,0,0") {
    angularDampingOverride=0.999
    angularSpringConstant=250
    bAngularSpring=true
    bOverrideAngularDamping=true
    bUseGravityOverride=true
    linearDampingOverride=0.99
    linearSpringConstant=550
    worldSpaceMinimumScale=0.1
    boundBone {boneName=moustache2_02_r}
    componentAppliedLinearAccClamp=<3000 3000 6000>
    componentLinearAccScale=<0.3 0.3 1>
    componentLinearVelScale=<0.2 0.2 0.1>
    gravityOverride {z=-980}
    physicsBodyDefinitions [{
      boundBone {boneName=moustache2_02_r}
      boxExtents=<10 10 10>
      constraintSetup {angularLimitsMax=<0 20 20>, angularLimitsMin=<0 -20 -20>, angularTarget=<1 0 0>}
      localJointOffset=<1 0.1 0.1>
     }]
   }
   n12 = Anim Dynamics - Bone: moustache2_01_l(ComponentPose=@n11, Alpha=1, ExternalForce="0,0,0") {
    angularDampingOverride=0.999
    angularSpringConstant=250
    bAngularSpring=true
    bOverrideAngularDamping=true
    bUseGravityOverride=true
    linearDampingOverride=0.99
    linearSpringConstant=550
    worldSpaceMinimumScale=0.1
    boundBone {boneName=moustache2_01_l}
    componentAppliedLinearAccClamp=<3000 3000 6000>
    componentLinearAccScale=<0.3 0.3 1>
    componentLinearVelScale=<0.2 0.2 0.1>
    gravityOverride {z=-980}
    physicsBodyDefinitions [{
      boundBone {boneName=moustache2_01_l}
      boxExtents=<10 10 10>
      constraintSetup {angularLimitsMax=<0 20 20>, angularLimitsMin=<0 -20 -20>, angularTarget=<1 0 0>}
      localJointOffset=<1 0.1 0.1>
     }]
   }
   n13 = Anim Dynamics - Bone: moustache2_02_l(ComponentPose=@n12, Alpha=1, ExternalForce="0,0,0") {
    angularDampingOverride=0.999
    angularSpringConstant=250
    bAngularSpring=true
    bOverrideAngularDamping=true
    bUseGravityOverride=true
    linearDampingOverride=0.99
    linearSpringConstant=550
    worldSpaceMinimumScale=0.1
    boundBone {boneName=moustache2_02_l}
    componentAppliedLinearAccClamp=<3000 3000 6000>
    componentLinearAccScale=<0.3 0.3 1>
    componentLinearVelScale=<0.2 0.2 0.1>
    gravityOverride {z=-980}
    physicsBodyDefinitions [{
      boundBone {boneName=moustache2_02_l}
      boxExtents=<10 10 10>
      constraintSetup {angularLimitsMax=<0 20 20>, angularLimitsMin=<0 -20 -20>, angularTarget=<1 0 0>}
      localJointOffset=<1 0.1 0.1>
     }]
   }
   Output Pose(Result=Component To Local(ComponentPose=@n13)) {layerGroup=None}
  ]}
 FE_Reactivity {
  entry=Idle
  path=AnimGraph.AnimGraphNode_StateMachine_1.FE_Reactivity
  ids=id
  nodes [
   AnimStateNode_0 = Idle()
   AnimStateNode_2 = FE_Reactive() {is=title AnimStateNode, stateEntered {guid=#1, notifyName=Entered FE_Reactive_Anim}}
   AnimStateTransitionNode_8 = AnimStateTransitionNode()
   AnimStateTransitionNode_9 = AnimStateTransitionNode() {bAutomaticRuleBasedOnSequencePlayerInState=true}
  ]
  links [
   {
    dst=Idle
    rule=AnimGraph.AnimGraphNode_StateMachine_1.FE_Reactivity.AnimStateTransitionNode_9.Transition
    src=FE_Reactive
    via=AnimStateTransitionNode_9
   }
   {
    dst=FE_Reactive
    rule=AnimGraph.AnimGraphNode_StateMachine_1.FE_Reactivity.AnimStateTransitionNode_8.Transition
    src=Idle
    via=AnimStateTransitionNode_8
   }
  ]
 }
 Idle {
  path=AnimGraph.AnimGraphNode_StateMachine_1.FE_Reactivity.AnimStateNode_0.Idle
  nodes [Output Animation Pose(Result=Sequence Player 'BiomeHose_Idle_PACK'() {sequence=/BRCosmetics/Accessories/FORT_Backpacks/Backpack_M_MED_BiomeHose/Animation/BiomeHose_Idle_PACK})]
 }
 FE_Reactive {
  path=AnimGraph.AnimGraphNode_StateMachine_1.FE_Reactivity.AnimStateNode_2.FE_Reactive
  nodes [Output Animation Pose(Result=Sequence Player 'BiomeHose_OpenChest_PACK'() {bLoopAnimation=false, sequence=/BRCosmetics/Accessories/FORT_Backpacks/Backpack_M_MED_BiomeHose/Animation/BiomeHose_OpenChest_PACK})]
 }
 Transition {
  path=AnimGraph.AnimGraphNode_StateMachine_1.FE_Reactivity.AnimStateTransitionNode_8.Transition
  nodes [Result(bCanEnterTransition=Get Play Reactive FE Anim() -> (Play Reactive FE Anim:bool))]
 }
 Transition {path=AnimGraph.AnimGraphNode_StateMachine_1.FE_Reactivity.AnimStateTransitionNode_9.Transition, nodes [Result()]}
 Idle and Reactive {
  entry=Idle
  path=AnimGraph.AnimGraphNode_StateMachine_2.Idle and Reactive
  ids=id
  nodes [
   AnimStateNode_0 = Reactive_OpenChest() {
    is=title AnimStateNode
    stateEntered {guid=#2}
    stateFullyBlended {guid=#3}
    stateLeft {guid=#4}
   }
   AnimStateNode_1 = Skydive Loop() {
    is=title AnimStateNode
    stateEntered {guid=#2}
    stateFullyBlended {guid=#3}
    stateLeft {guid=#4}
   }
   AnimStateNode_2 = Idle() {
    stateEntered {guid=#5}
    stateFullyBlended {guid=#6}
    stateLeft {guid=#7}
   }
   AnimStateNode_4 = Skydive() {
    is=title AnimStateNode
    stateEntered {guid=#2}
    stateFullyBlended {guid=#3}
    stateLeft {guid=#4}
   }
   AnimStateTransitionNode_2 = AnimStateTransitionNode() {
    crossfadeDuration=0.3
    transitionEnd {guid=#8}
    transitionInterrupt {guid=#9}
    transitionStart {guid=#10}
   }
   AnimStateTransitionNode_4 = AnimStateTransitionNode() {bAllowInertializationForSelfTransitions=true}
   AnimStateTransitionNode_5 = AnimStateTransitionNode() {bAllowInertializationForSelfTransitions=true, bAutomaticRuleBasedOnSequencePlayerInState=true}
   AnimStateTransitionNode_6 = AnimStateTransitionNode() {bAllowInertializationForSelfTransitions=true, bAutomaticRuleBasedOnSequencePlayerInState=true}
   AnimStateTransitionNode_8 = AnimStateTransitionNode() {bAllowInertializationForSelfTransitions=true}
   AnimStateTransitionNode_9 = AnimStateTransitionNode() {bAllowInertializationForSelfTransitions=true}
  ]
  links [
   {
    dst=Reactive_OpenChest
    rule=AnimGraph.AnimGraphNode_StateMachine_2.Idle and Reactive.AnimStateTransitionNode_4.Transition
    src=Idle
    via=AnimStateTransitionNode_4
   }
   {
    dst=Skydive
    rule=AnimGraph.AnimGraphNode_StateMachine_2.Idle and Reactive.AnimStateTransitionNode_2.Transition
    src=Idle
    via=AnimStateTransitionNode_2
   }
   {
    dst=Idle
    rule=AnimGraph.AnimGraphNode_StateMachine_2.Idle and Reactive.AnimStateTransitionNode_5.Transition
    src=Reactive_OpenChest
    via=AnimStateTransitionNode_5
   }
   {
    dst=Idle
    rule=AnimGraph.AnimGraphNode_StateMachine_2.Idle and Reactive.AnimStateTransitionNode_9.Transition
    src=Skydive
    via=AnimStateTransitionNode_9
   }
   {
    dst=Skydive Loop
    rule=AnimGraph.AnimGraphNode_StateMachine_2.Idle and Reactive.AnimStateTransitionNode_6.Transition
    src=Skydive
    via=AnimStateTransitionNode_6
   }
   {
    dst=Idle
    rule=AnimGraph.AnimGraphNode_StateMachine_2.Idle and Reactive.AnimStateTransitionNode_8.Transition
    src=Skydive Loop
    via=AnimStateTransitionNode_8
   }
  ]
 }
 Reactive_OpenChest {
  path=AnimGraph.AnimGraphNode_StateMachine_2.Idle and Reactive.AnimStateNode_0.Reactive_OpenChest
  nodes [Output Animation Pose(Result=Sequence Player 'BiomeHose_OpenChest_PACK'() {bLoopAnimation=false, sequence=&BiomeHose_OpenChest_PACK})]
 }
 Skydive Loop {
  path=AnimGraph.AnimGraphNode_StateMachine_2.Idle and Reactive.AnimStateNode_1.Skydive Loop
  nodes [
   n1 = Random Sequence Player!() -> (Pose:Pose Link Structure) {
    is=title AnimGraphNode_RandomPlayer
    entries [
     {
      chanceToPlay=0.2
      maxLoopCount=0
      maxPlayRate=1
      minLoopCount=0
      minPlayRate=1
      sequence=/BRCosmetics/Accessories/FORT_Backpacks/Backpack_M_MED_BiomeHose/Animation/BiomeHose_SkydiveLoop_PACK
      blendIn {blendOption=HermiteCubic, blendTime=0.2, customCurve=null}
     }
     {
      chanceToPlay=0.4
      maxLoopCount=0
      maxPlayRate=1
      minLoopCount=0
      minPlayRate=1
      sequence=/BRCosmetics/Accessories/FORT_Backpacks/Backpack_M_MED_BiomeHose/Animation/BiomeHose_SkydiveLoop2_PACK
      blendIn {blendOption=HermiteCubic, blendTime=0.2, customCurve=null}
     }
     {
      chanceToPlay=0.2
      maxLoopCount=0
      maxPlayRate=1
      minLoopCount=0
      minPlayRate=1
      sequence=/BRCosmetics/Accessories/FORT_Backpacks/Backpack_M_MED_BiomeHose/Animation/BiomeHose_SkydiveLoopL_PACK
      blendIn {blendOption=HermiteCubic, blendTime=0.2, customCurve=null}
     }
     {
      chanceToPlay=0.2
      maxLoopCount=0
      maxPlayRate=1
      minLoopCount=0
      minPlayRate=1
      sequence=/BRCosmetics/Accessories/FORT_Backpacks/Backpack_M_MED_BiomeHose/Animation/BiomeHose_SkydiveLoopR_PACK
      blendIn {blendOption=HermiteCubic, blendTime=0.2, customCurve=null}
     }
    ]
   }
   Output Animation Pose(Result=@n1)
  ]
 }
 Idle {
  path=AnimGraph.AnimGraphNode_StateMachine_2.Idle and Reactive.AnimStateNode_2.Idle
  nodes [Output Animation Pose(Result=Sequence Player 'BiomeHose_Idle_PACK'() {sequence=&BiomeHose_Idle_PACK})]
 }
 Skydive {
  path=AnimGraph.AnimGraphNode_StateMachine_2.Idle and Reactive.AnimStateNode_4.Skydive
  nodes [
   n1 = Sequence Player 'BiomeHose_SkydiveIntro_PACK'() {bLoopAnimation=false, sequence=/BRCosmetics/Accessories/FORT_Backpacks/Backpack_M_MED_BiomeHose/Animation/BiomeHose_SkydiveIntro_PACK}
   Output Animation Pose(Result=@n1)
  ]
 }
 Transition {
  path=AnimGraph.AnimGraphNode_StateMachine_2.Idle and Reactive.AnimStateTransitionNode_2.Transition
  nodes [Result(bCanEnterTransition=Get bIsSkyDiving())]
 }
 Transition {
  path=AnimGraph.AnimGraphNode_StateMachine_2.Idle and Reactive.AnimStateTransitionNode_4.Transition
  nodes [Result(bCanEnterTransition=Get Reactive OpenChest() -> (Reactive OpenChest:bool))]
 }
 Transition {path=AnimGraph.AnimGraphNode_StateMachine_2.Idle and Reactive.AnimStateTransitionNode_5.Transition, nodes [Result()]}
 Transition {path=AnimGraph.AnimGraphNode_StateMachine_2.Idle and Reactive.AnimStateTransitionNode_6.Transition, nodes [Result()]}
 Transition {
  path=AnimGraph.AnimGraphNode_StateMachine_2.Idle and Reactive.AnimStateTransitionNode_8.Transition
  nodes [Result(bCanEnterTransition=NOT Boolean(A=Get bIsSkyDiving()))]
 }
 Transition {
  path=AnimGraph.AnimGraphNode_StateMachine_2.Idle and Reactive.AnimStateTransitionNode_9.Transition
  nodes [Result(bCanEnterTransition=NOT Boolean(A=Get bIsSkyDiving()))]
 }
 EventGraph {nodes [
   Event OnCharacterCustomizationLoaded()
   > n1 = Cast To FortPlayerPawn(Object=TryGetFortPawn())
   > n2 = Branch()
   > Do Once(Reset=@n6.then)
   > n3 = Delay(Duration=14)
   Set Play Reactive FE Anim(execute=[@n3.then, @n5.then], Play Reactive FE Anim:bool=true) -> (Output_Get:bool)
   > n4 = Delay(Duration=1)
   Set Play Reactive FE Anim(execute=[@n4.then, @n16.then], Play Reactive FE Anim=false)
   > n5 = Delay(Duration=18)
   n6 = Set Play Reactive FE Anim(execute=@n2.else, Play Reactive FE Anim=false)
   Event BlueprintUpdateAnimation()
   > n7 = Sequence()
   Set NegPhysicsAlpha(
    execute=@n7.then_1
    NegPhysicsAlpha:double=MapRangeClamped(Value=GetCurveValue(CurveName=PhysicsOff), InRangeA=0, InRangeB=1, OutRangeA=1, OutRangeB=0)
   ) -> (Output_Get:double)
   n8 = GetOwnerSkeletalMeshComponent()
   n9 = BreakRotator(InRot=BreakTransform(InTransform=GetDeltaTransformFromRefPose(self=@n8, BoneName=spine_05, BaseName=spine_01)).Rotation)
   n10 = MapRangeClamped(Value=@n9.Yaw, InRangeA=-10, InRangeB=-50, OutRangeB=50)
   Set backpack_rotate(execute=@n7.then_0, backpack_rotate=MakeRotator(Roll=@n10, Pitch=float / float(A=@n10, B=10)))
   > Set backpack_translation(backpack_translation=MakeVector(Y=MapRangeClamped(Value=@n9.Yaw, InRangeA=-15, InRangeB=-50, OutRangeB=-11)))
   n11 = GetCurveValue(CurveName=L_blink_pose)
   n12 = GetCurveValue(CurveName=R_blink_pose)
   n13 = OR Boolean(
    A="Not Equal (Float)"(A=@n12, B=Get PrevRBlinkPose() -> (PrevRBlinkPose:float))
    B="Not Equal (Float)"(A=@n11, B=Get PrevLBlinkPose() -> (PrevLBlinkPose:float))
   )
   n14 = MakeVector2D(
    X=float - float(A=GetCurveValue(CurveName=look_left_pose), B=GetCurveValue(CurveName=look_right_pose))
    Y=float - float(A=GetCurveValue(CurveName=look_up_pose), B=GetCurveValue(CurveName=look_down_pose))
   )
   n15 = "Not Equal (Vector2D)"!(
    A:Vector 2D Structure=@n14
    B:Vector 2D Structure=Get PrevLookDirection() -> (PrevLookDirection:Vector 2D Structure)
    ErrorTolerance:float=0.025
   ) -> (ReturnValue:bool=~) {is=title K2Node_PromotableOperator}
   > SetScalarParameterValueOnMaterials!(
    execute
    self:Mesh Component Object Reference=GetOwningComponent()
    ParameterName:Name=HasLookDirectionChanged
    ParameterValue:float="To Float (Boolean)"(InBool=OR Boolean(A=@n15, B=@n13))
   ) -> (then)
   > Set PrevLookDirection(PrevLookDirection:Vector 2D Structure=@n14) -> (Output_Get:Vector 2D Structure)
   > Set PrevRBlinkPose(PrevRBlinkPose:float=@n12) -> (Output_Get:float)
   > Set PrevLBlinkPose(PrevLBlinkPose:float=@n11) -> (Output_Get:float)
   n16 = AnimNotify_FE_ReactiveAnim() {is=title K2Node_Event}
   n17 = Event BlueprintProcessAnimInputEvent()
   > Branch(Condition=IsNamed(self=@n17.AnimInputEvent, InEventName=aie.ChestOpened))
   > Set Reactive OpenChest(Reactive OpenChest:bool=true) -> (Output_Get:bool)
   > RetriggerableDelay(Duration=0.2)
   > Set Reactive OpenChest(Reactive OpenChest=false)
   TryGetPawnOwner()
   GetModifierForPartType!(self:Fort Player Pawn Object Reference=@n1, PartType:EFortCustomPartType Enum=Backpack) -> (ReturnValue:Custom Character Part Modifier Object Reference)
  ]}
]
variables [
 backpack_rotate {type=Rotator}
 backpack_translation {type=Vector}
 NegPhysicsAlpha {type=number, default=1}
 Play Reactive FE Anim {type=boolean}
 PrevLBlinkPose {type=number}
 PrevLookDirection {type=Vector2D, default=<0 0>}
 PrevRBlinkPose {type=number}
 Reactive OpenChest {type=boolean}
]
